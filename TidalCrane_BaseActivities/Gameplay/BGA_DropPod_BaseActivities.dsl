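; Drop pod Blueprint for TidalCrane base activities that customizes a parent drop pod's look and loot.
; On BeginPlay it caches the current playlist once ready; SpawnLoot uses its max team size to index a loot tier group array, spawns pickups and adds a source tag to each.
; SetupDropPodMaterials creates dynamic material instances on the mesh's first two slots to tint screen and glass green.
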(event EventBeginPlay
  (|CallParentFunction)
  (bind (_gamestate _playlist) (Class|FortAsyncActionCurrentPlaylistReady|PlaylistDataReadyAsync)
    (:OnReady
      (Variables|Default|SetPlaylist _playlist))))

(event Collision|EventActorBeginOverlap
  (params (OtherActor :type Actor%20Object%20Reference))
  (|CallParentFunction OtherActor))

(event EventTick
  (params (DeltaSeconds :type Float%20%28single-precision%29))
  (|CallParentFunction DeltaSeconds))

(fn SetupDropPodMaterials
  (bind _returnvalue (Rendering|Material|GetMaterial (Variables|Debug|GetStaticMeshComponent) 1))
  (bind _staticmeshcomponent (Variables|Debug|GetStaticMeshComponent))
  (bind _returnvalue_2 (Rendering|Material|GetMaterial _staticmeshcomponent 0))
  ; To Change the Screen and Glass Color
  (|CallParentFunction)
  (Utilities|IsValid _staticmeshcomponent
    (:Is%20Valid
      (Utilities|IsValid _returnvalue_2
        (:Is%20Valid
          (bind _returnvalue_3 (Rendering|Material|CreateDynamicMaterialInstance _staticmeshcomponent 0 _returnvalue_2 "None" :!class /Script/Engine.PrimitiveComponent))
          (Rendering|Material|SetVectorParameterValue _returnvalue_3 "DefaultScreenTint" "(R=0.000000,G=0.597202,B=0.135633,A=1.000000)" :!class /Script/Engine.MaterialInstanceDynamic)
          (Rendering|Material|SetVectorParameterValue _returnvalue_3 "EmissiveTint" "(R=0.000000,G=0.597202,B=0.135633,A=1.000000)" :!class /Script/Engine.MaterialInstanceDynamic)
          (Utilities|IsValid _returnvalue
            (:Is%20Valid
              (bind _returnvalue_4 (Rendering|Material|CreateDynamicMaterialInstance _staticmeshcomponent 1 _returnvalue "None" :!class /Script/Engine.PrimitiveComponent))
              (Rendering|Material|SetVectorParameterValue _returnvalue_4 "Tint_Color" "(R=0.000000,G=0.597202,B=0.135633,A=1.000000)" :!class /Script/Engine.MaterialInstanceDynamic))
            (:Is%20Not%20Valid
              _)))
        (:Is%20Not%20Valid
          _))
      ;[0.4115 0 0 0.5] Warning: StaticMeshComponent or Material returned Invalid.
      (Development|PrintString (Utilities|String|Append (Utilities|GetDisplayName self) (LocalVariables|Getcomponent%20not%20found%20text))))
    (:Is%20Not%20Valid
      _)))

(fn SpawnLoot
  ; Check the playlist size to see how much loot to grant
  (bind _asfort_playlist_athena (Utilities|Casting|CastToFortPlaylistAthena (Variables|Default|GetPlaylist)))
  (bind _result (ArrayMacros|SafeArrayGet (Variables|Default|GetPlaylistLTG) (- (Playlist|GetMaxTeamSize (Fortnite|Athena|GameState|GetGameStateAthena) _asfort_playlist_athena) 1))
    (:%20%20
      (bind _returnvalue (Inventory|SpawnPickupInWorldWithLootTier _result (CallFunction|GetLootSpawnLocation)))
      (for _array_element _returnvalue
        (Class|FortPickup|AddSpawnSourceTags _array_element (Variables|Default|GetPickUp%20Source%20Tag))))))

(fn ConstructionScript
  (|CallParentFunction))
