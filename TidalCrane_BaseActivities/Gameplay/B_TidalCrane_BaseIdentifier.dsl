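; B_TidalCrane_BaseIdentifier is an actor Blueprint that hands base locator data to the TidalCrane military outposts system.
; On BeginPlay it gets the Athena game state and looks up its MilitaryOutposts GameStateComponent.
; On the authority only, if that component is valid, it calls OnReadytoStart with its BaseLocators and SkipSafeZoneChecks variables.
; The construction script is empty.
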(event EventBeginPlay
  (bind _returnvalue (Fortnite|Athena|GameState|GetGameStateAthena))
  (bind _returnvalue_2 (Actor|GetComponentbyClass _returnvalue "/TidalCrane_BaseActivities/Gameplay/B_TidalCrane_MilitaryOutposts_GameStateComponent.B_TidalCrane_MilitaryOutposts_GameStateComponent_C"))
  (Networking|SwitchHasAuthority
    (:Authority
      (Utilities|IsValid _returnvalue_2
        (:Is%20Valid
          (Class|BTidalCraneMilitaryOutpostsGameStateComponent|OnReadytoStart _returnvalue_2 (Variables|Default|GetBaseLocators) (Variables|Default|GetSkipSafeZoneChecks)))))))

(fn ConstructionScript)
