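; B_TidalCrane_BaseLocator is an actor that marks a TidalCrane base, tracking its type and state.
; Authority-set state replicates to switch minimap icon and legend between normal and infested; type changes broadcast through the playspace event router, also at BeginPlay when preset.
; It checks whether it lies inside a phase's safe zone and reserves or releases Living World AI density for base defense.
; Tick drives a material timer from server time.
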(event EventBeginPlay
  ; Special case for test levels, which don't use the GameStateComponent manager. If a BaseLocator is hardcoded to a specific base type, send the "Base Type Has Been Set" event message at startup
  (if (Utilities|Enum|NotEqual(Enum) (Variables|Default|GetBaseType) "NewEnumerator4")
    (CallFunction|SendBaseTypeEvent)))

(event EventTick
  (params (DeltaSeconds :type Float%20%28single-precision%29))
  (CallFunction|UpdateTimer))

(fn ActivateKlaxon)

(fn ClearTimer
  (Variables|Default|SetCurrentDynamicMaterial 0)
  (Actor|Tick|SetActorTickEnabled false))

(fn IsValidForPhase
  (params (Phase :type Integer) (SkipSafeZoneCheck :type Boolean))
  (bind _returnvalue (Fortnite|Athena|GameState|GetGameStateAthena))
  (bind _mapinfo (Variables|/Script/FortniteGame.FortGameStateAthena|GetMapInfo _returnvalue))
  (if SkipSafeZoneCheck
    (return true)
    (else
      (Utilities|IsValid _mapinfo
        (:Is%20Valid
          (bind (_outlocation _returnvalue_2) (SafeZone|GetSafeZoneLocation Phase))
          (if _returnvalue_2
            (return (<= (Math|Vector2D|DistanceSquared2D (Math|Conversions|ToVector2D(Vector) _outlocation) (Math|Conversions|ToVector2D(Vector) (Transformation|GetActorLocation))) (Math|Float|Square (SafeZones|GetSafeZoneRadius _mapinfo Phase))))
            (else
              (return false))))
        (:Is%20Not%20Valid
          (Fortnite|FortShippingLog self "Error: No GameState Map Info!
Unable to determine if BaseLocator is safe" true)
          (return false))))))

(fn OnRep_BaseState
  (bind _output_get (LocalVariables|SetCurrent%20State (Utilities|Select :Index (Variables|Default|GetBaseState) :NewEnumerator0 (Variables|MinimapIcon|GetMinimapIcon_Normal) :NewEnumerator1 (Variables|MinimapIcon|GetMinimapIcon_Normal) :NewEnumerator2 (Variables|MinimapIcon|GetMinimapIcon_Infested) :NewEnumerator3 (Variables|MinimapIcon|GetMinimapIcon_Infested) :NewEnumerator4 (Variables|MinimapIcon|GetMinimapIcon_Infested) :NewEnumerator5 (Variables|MinimapIcon|GetMinimapIcon_Normal) :NewEnumerator6 (Variables|MinimapIcon|GetMinimapIcon_Normal))))
  (Class|AthenaSpecialActorComponent|SetMiniMapIcon (Variables|Default|GetAthenaSpecialActor) _output_get)
  (Class|AthenaSpecialActorComponent|SetMiniMapIconMapLegendText (Variables|Default|GetAthenaSpecialActor) (Utilities|Select :Index (Variables|Default|GetBaseState) :NewEnumerator0 (Variables|MinimapIcon|GetMinimapLegend_Normal) :NewEnumerator1 (Variables|MinimapIcon|GetMinimapLegend_Normal) :NewEnumerator2 (Variables|MinimapIcon|GetMinimapLegend_Infested) :NewEnumerator3 (Variables|MinimapIcon|GetMinimapLegend_Infested) :NewEnumerator4 (Variables|MinimapIcon|GetMinimapLegend_Infested) :NewEnumerator5 (Variables|MinimapIcon|GetMinimapLegend_Normal) :NewEnumerator6 (Variables|MinimapIcon|GetMinimapLegend_Normal)))
  (CallFunction|ClearTimer))

(fn OnRep_BaseType
  (CallFunction|SendBaseTypeEvent))

(fn ReleaseDensity
  (Class|FortAthenaLivingWorldManager|UnregisterDensityReservoir (Class|FortAthenaLivingWorldManager|GetLivingWorldManager) self))

(fn ReserveDensityForBaseDefense
  (for _index (range (Math|Float|Truncate (ScalableFloat|GetValueatLevel (Variables|Hotfix|GetNumBaseDefenseInstanceSpawners))))
    (LocalVariables|SetAmountToReserve (+ (LocalVariables|GetAmountToReserve) (ScalableFloat|GetValueatLevel (Variables|Hotfix|GetAI%20Resevoir%20Density) (Math|Conversions|ToFloat(Integer) _index)))))
  (Class|FortAthenaLivingWorldManager|RegisterDensityReservoir (Class|FortAthenaLivingWorldManager|GetLivingWorldManager) self (LocalVariables|GetAmountToReserve) (ScalableFloat|GetValueatLevel (Variables|Hotfix|GetAI%20Resevoir%20Range))))

(fn SendBaseTypeEvent
  (bind _returnvalue (GameplayEventRouter|GetEventRouter "/Script/FortniteGame.FortPlayspace" self))
  (GameplayEventRouter|BroadcastEvent _returnvalue (Utilities|Struct|MakeTidalCraneBaseTypeData self (Variables|Default|GetBaseType)) 0))

(fn SetBaseMinimapIcon
  (params (Current%20Base%20State :type TidalCrane_Bases_Enum%20Enum :default "NewEnumerator0"))
  (Creative|Networking|Authority/Cosmetic
    (:Authority
      (Variables|Default|SetBaseState Current%20Base%20State)
      (Variables|Default|SetServerTimeForStateStart (Fortnite|Sync|Time|GetServerWorldTimeSeconds)))))

(fn SetBaseType
  (params (BaseType :type TidalCrane_BaseTypes%20Enum :default "NewEnumerator4"))
  (Networking|SwitchHasAuthority
    (:Authority
      (Variables|Default|SetBaseType BaseType))))

(fn UpdateTimer
  (LocalVariables|SetPercentElapsed (/ (- (Fortnite|Sync|Time|GetServerWorldTimeSeconds) (Variables|Default|GetServerTimeForStateStart)) 3e+01))
  (Rendering|Material|SetScalarParameterValue (Variables|Default|GetCurrentDynamicMaterial) "TimeProgress" (LocalVariables|GetPercentElapsed) :!class /Script/Engine.MaterialInstanceDynamic)
  (if (>= (LocalVariables|GetPercentElapsed) 1.0)
    (Actor|Tick|SetActorTickEnabled false)))

(fn ConstructionScript)
