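; Base spawn manager for TidalCrane military base activities, with hooks for child classes and quest and team rank-point payouts on completion.
; A box volume finds the base's NPCs, locator and klaxons, and tracks players for server participation and client overlap events.
; Rewards spawn a pinged chest and temporarily leash NPCs there with cheering smart objects, saving patrol paths for later restoration.
; A server timer disables the activity when the storm engulfs it.
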(event EventBeginPlay
  (CallFunction|GetInitialOverlaps)
  (for _array_element (Variables|Default|GetNPC_CheeringSmartObjects)
    (SmartObject|SetSmartObjectEnabled _array_element))
  (Networking|SwitchHasAuthority
    (:Authority
      (bind _returnvalue (Utilities|Time|SetTimerbyEvent (EventDispatchers|CreateEvent :!event Custom|DoStormCheck) (ScalableFloat|GetValueasDouble (Variables|Default|GetHFRow%20StormCheckTimer)) true))
      (Variables|Default|SetStormCheckTimerHandle _returnvalue))))

(event OnComponentBeginOverlap(Box)
  (params (OverlappedComponent :type Primitive%20Component%20Object%20Reference) (OtherActor :type Actor%20Object%20Reference) (OtherComp :type Primitive%20Component%20Object%20Reference) (OtherBodyIndex :type Integer) (bFromSweep :type Boolean) (SweepResult :type Hit%20Result%20Structure%20%28by%20ref%29))
  ; Overlap Handling
  (bind _asfort_player_pawn (Utilities|Casting|CastToFortPlayerPawn OtherActor))
  (if (Pawn|IsPlayerControlled _asfort_player_pawn)
    (Utilities|FlowControl|Sequence
      (:then_0
        (Networking|SwitchHasAuthority
          (:Authority
            ;; Add the player to a list of quest participants on the server
            (if (not (Utilities|Set|ContainsItem (Variables|Default|GetQuestParticipants) _asfort_player_pawn))
              (Utilities|Set|Add (Variables|Default|GetQuestParticipants) _asfort_player_pawn)))))
      (:then_1
        (if (not (Utilities|Set|ContainsItem (Variables|Default|GetOverlappingPlayers) _asfort_player_pawn))
          (Utilities|Set|Add (Variables|Default|GetOverlappingPlayers) _asfort_player_pawn)
          (CallFunction|SendOverlapEvent _asfort_player_pawn true))))))

(event OnComponentEndOverlap(Box)
  (params (OverlappedComponent :type Primitive%20Component%20Object%20Reference) (OtherActor :type Actor%20Object%20Reference) (OtherComp :type Primitive%20Component%20Object%20Reference) (OtherBodyIndex :type Integer))
  (bind _overlappingplayers (Variables|Default|GetOverlappingPlayers))
  ; Overlap Handling
  (bind _asfort_player_pawn (Utilities|Casting|CastToFortPlayerPawn OtherActor))
  (Utilities|FlowControl|Sequence
    (:then_0
      (Networking|SwitchHasAuthority
        (:Authority
          (Utilities|Set|Remove (Variables|Default|GetQuestParticipants) _asfort_player_pawn))))
    (:then_1
      (if (Utilities|Set|ContainsItem _overlappingplayers _asfort_player_pawn)
        (Utilities|Set|Remove _overlappingplayers _asfort_player_pawn)
        (CallFunction|SendOverlapEvent _asfort_player_pawn)))))

(event Custom|SendOverlapEvent
  (params (OverlappingActor :type Fort%20Player%20Pawn%20Object%20Reference) (IsOverlapping :type Boolean))
  ; Overlap Handling
  (Networking|SwitchHasAuthority
    (:Authority
      (if (or (== (Utilities|Set|Length (Variables|Default|GetOverlappingPlayers)) 0) (== (Utilities|Set|Length (Variables|Default|GetOverlappingPlayers)) 1))
        (CallFunction|OnPlayerPresenceChanged)))
    (:Remote
      (bind _returnvalue (GameplayEventRouter|GetEventRouter "/Script/FortniteGame.FortPlayspace" self))
      (GameplayEventRouter|BroadcastStatefulEvent _returnvalue (Utilities|Struct|MakeTidalCraneDefenseOverlapStatefulEvent (Utilities|ResolveSoftReference self) (Variables|Settings|GetBaseType) (Utilities|ResolveSoftReference OverlappingActor) IsOverlapping) OverlappingActor))))

(event Custom|Server_PingReward
  (params (ActorToPing :type Actor%20Object%20Reference))
  (Networking|SwitchHasAuthority
    (:Authority
      (bind _outactors (Collision|SphereOverlapActors (Transformation|GetActorLocation) 2e+04 (Utilities|Array|MakeArray "ObjectTypeQuery3") "/Game/Athena/PlayerPawn_Athena.PlayerPawn_Athena_C"))
      (for _array_element _outactors
        (Utilities|IsValid _array_element
          (:Is%20Valid
            (Utilities|IsValid (Fortnite|GetControllerfromActor _array_element)
              (:Is%20Valid
                (Utilities|IsValid (Actor|GetComponentbyClass (Fortnite|GetControllerfromActor _array_element) "/Script/FortniteGame.AthenaMarkerComponent")
                  (:Is%20Valid
                    (MarkerComponent|MarkActoronServer (Actor|GetComponentbyClass (Fortnite|GetControllerfromActor _array_element) "/Script/FortniteGame.AthenaMarkerComponent") ActorToPing true false)))))))))))

(event Custom|OnNPCFinishedGrantingLoot
  (params (NPC :type Fort%20Player%20Pawn%20Object%20Reference))
  (Networking|SwitchHasAuthority
    (:Authority
      (CallFunction|ServerPingReward NPC))))

(event Custom|SpawnChestRewards
  (bind _rewardchestspawnlocation (Variables|Default|GetRewardChestSpawnLocation))
  (bind (_returnvalue_location _returnvalue_rotation _returnvalue_scale) (Math|Transform|BreakTransform (Transformation|GetActorTransform _rewardchestspawnlocation)))
  (Utilities|FlowControl|Delay (ScalableFloat|GetValueatLevel (Variables|Default|GetRewardDelay)))
  (Networking|SwitchHasAuthority
    (:Authority
      (Utilities|IsValid _rewardchestspawnlocation
        (:Is%20Valid
          (bind _returnvalue (Game|SpawnActorfromClass (Variables|Settings|GetRewardChestClass) (Math|Transform|MakeTransform (+ _returnvalue_location "0.000000,0.000000,8000.000000") _returnvalue_rotation _returnvalue_scale) "Undefined" "MultiplyWithRoot" 0 0 0 0))
          (Utilities|IsValid _returnvalue
            (:Is%20Valid
              (Variables|/Script/Engine.Actor|SetInstigator (Variables|Actor|GetInstigator) _returnvalue)
              (CallFunction|ServerPingReward _rewardchestspawnlocation)
              (for _array_element (Variables|Default|GetNPCs)
                (CallFunction|SetNPCLeashtoReward _array_element) ; Tell NPCs to gather at the drop location
              )
              (CallFunction|DelayResetBehavioronNPCs))))))))

(event Custom|DelayResetBehaviorOnNPCs
  ; Reset the NPC patrol paths, immunity, and faction after the given delay in Reset Behavior Time
  (Utilities|Time|SetTimerbyEvent (delegate Custom|NPCResetTimeUp) (ScalableFloat|GetValueatLevel (Variables|NPCs|GetNPCResetBehaviorTime))))

(event Custom|NPCResetTimeUp
  ; Reset the NPC patrol paths, immunity, and faction after the given delay in Reset Behavior Time
  (for _array_element (Variables|Default|GetNPCs)
    (CallFunction|RestoreNPCPatrolPath _array_element)
    (CallFunction|ApplyorRemoveEffectsonNPC true _array_element)))

(event Custom|DoStormCheck
  ; Check if MilitaryActivity is within storm (if Box center is inside storm, call ConsumedByStorm for child classes to handle)
  (Networking|SwitchHasAuthority
    (:Authority
      (bind _returnvalue (SafeZone|IsLocationinSafeZone (Transformation|GetWorldLocation (Variables|Default|GetBox))))
      (if (not _returnvalue)
        (Utilities|Time|ClearandInvalidateTimerbyHandle (Variables|Default|GetStormCheckTimerHandle))
        (CallFunction|OnConsumedbyStorm)))))

(event EventEndPlay
  (params (EndPlayReason :type EEndPlayReason%20Enum :default "Destroyed"))
  (bind _stormchecktimerhandle (Variables|Default|GetStormCheckTimerHandle))
  (if (Utilities|Time|IsValidTimerHandle _stormchecktimerhandle)
    (Utilities|Time|ClearandInvalidateTimerbyHandle _stormchecktimerhandle)))

(event Custom|TidalCraneForceStormFailure
  ; ke * TidalCraneForceStormFailure - Sends message to all bases as if they were engulfed by the storm
  (Networking|SwitchHasAuthority
    (:Authority
      (CallFunction|OnConsumedbyStorm))))

(fn ApplyOrRemoveEffectsOnNPC
  (params (bShouldRemove :type Boolean) (NPC :type Fort%20Player%20Pawn%20Object%20Reference))
  (bind _returnvalue (Ability|GetFortAbilitySystemComponent NPC))
  ;[0 0 0] Apply or remove the GE that gives NPCs invulnerability to player damage, and increases damage done/taken to bugs
  (Utilities|IsValid NPC
    (:Is%20Valid
      (Utilities|IsValid _returnvalue
        (:Is%20Valid
          (if bShouldRemove
            (GameplayEffects|RemoveActiveEffectswithGrantedTags _returnvalue (Ability|GameplayEffect|GetGrantedTags (Variables|NPCs|GetNPCBaseActivityGE) :!class /Script/GameplayAbilities.AbilitySystemBlueprintLibrary))
            (else
              (GameplayEffects|ApplyGameplayEffectToSelf :self _returnvalue :GameplayEffectClass (Variables|NPCs|GetNPCBaseActivityGE) :Level 0.0))))))))

(fn ChangeFactionOnNPC
  (params (bRestoreDefaultBehavior :type Boolean) (NPC :type Fort%20Player%20Pawn%20Object%20Reference))
  ;[0 0 0] Add or remove the NPC from the correct base activity faction depending on if they should return to default NPC behavior
  (if bRestoreDefaultBehavior
    ;; If default behavior should be restored, remove them from the base active faction and add to the not active faction
    (Affiliation|RemovefromFaction NPC "(TagName=\"Athena.Faction.BaseActivity.Active\")")
    (Affiliation|AddtoFaction NPC "(TagName=\"Athena.Faction.BaseActivity.BaseNotActive\")")
    (else
      ;; If default behavior bool isn't checked, remove them from the not active faction and add to the base active faction
      (Affiliation|RemovefromFaction NPC "(TagName=\"Athena.Faction.BaseActivity.BaseNotActive\")")
      (Affiliation|AddtoFaction NPC "(TagName=\"Athena.Faction.BaseActivity.Active\")"))))

(fn CheckIfPlayersIsOverlapping
  (params (Spectator :type Fort%20Player%20Pawn%20Object%20Reference))
  (return (Utilities|Set|ContainsItem (Variables|Default|GetOverlappingPlayers) Spectator)))

(fn GetInitialOverlaps
  (bind _outactors (Collision|ComponentOverlapActors (Variables|Default|GetBox) (Transformation|GetWorldTransform (Variables|Default|GetBox)) (Utilities|Array|MakeArray "ObjectTypeQuery3") "/Script/FortniteGame.FortPlayerPawn"))
  (Utilities|FlowControl|Sequence
    (:then_0
      (for _array_element _outactors
        ;[0.15 0.0352 0.1161 0.5] Server find NPCs
        (bind (_bisnpc _npc) (CallFunction|ProcessNPC _array_element))
        (if _bisnpc
          (Utilities|Array|AddUnique (Variables|Default|GetNPCs) _npc)))
      ;[0.1442 0.0322 0.15 0.5] Server find BaseLocator
      (bind _outactors_2 (Actor|GetAllActorsOfClass "/TidalCrane_BaseActivities/Gameplay/B_TidalCrane_BaseLocator.B_TidalCrane_BaseLocator_C"))
      (bind _array_element_2 (Utilities|Array|ForEachLoopwithBreak _outactors_2)
        (:LoopBody
          (bind _binarea (CallFunction|IsActorinEventArea _array_element_2))
          (if _binarea
            (Variables|Default|SetBaseLocator _array_element_2)))
        (:Completed
          _))
      ;[0.1442 0.0322 0.15 0.5] Server find Klaxon actors
      (bind _outactors_3 (Actor|GetAllActorsOfClass "/HeistBasesGameplay/Items/Environmental/BP_Lockdown_Prop_Parent.BP_Lockdown_Prop_Parent_C"))
      (for _array_element_3 _outactors_3
        (bind _binarea_2 (CallFunction|IsActorinEventArea _array_element_3))
        (if _binarea_2
          (Utilities|Array|Add (Variables|Default|GetKlaxons) _array_element_3))))
    (:then_1
      ;[0.07393 0.09561 0.3217 0.5] Detect players in the volume (send Overlap event on client)
      (for _array_element_4 _outactors
        ;; Check for Players
        (bind _asfort_player_pawn (Utilities|Casting|CastToFortPlayerPawn _array_element_4))
        (if (Pawn|IsPlayerControlled _asfort_player_pawn)
          (Utilities|Set|Add (Variables|Default|GetOverlappingPlayers) _asfort_player_pawn)
          (CallFunction|SendOverlapEvent _asfort_player_pawn true))))))

(fn IsActorInEventArea
  (params (Actor :type Actor%20Object%20Reference))
  (bind _returnvalue_location (Math|Transform|BreakTransform (Transformation|GetWorldTransform (Variables|Default|GetBox))))
  (return (Math|Geometry|IsPointinBox (Transformation|GetActorLocation Actor) _returnvalue_location (Components|Box|GetScaledBoxExtent (Variables|Default|GetBox)))))

(fn LeashNPCInPlace
  (params (NPC :type Fort%20Player%20Pawn%20Object%20Reference))
  (Utilities|IsValid NPC
    (:Is%20Valid
      (bind _asfort_athena_aibot_controller (Utilities|Casting|CastToFortAthenaAIBotController (Pawn|GetController NPC)))
      (Class|BPFLAICommonFunctions|SetRadialLeash _asfort_athena_aibot_controller (Transformation|GetActorLocation NPC) 1e+02 1e+02)
      (Utilities|IsValid (Actor|GetComponentbyClass _asfort_athena_aibot_controller "/Script/FortniteAI.FortAthenaNpcPatrollingComponent")
        (:Is%20Valid
          (if (not (Utilities|Map|Contains (Variables|Default|GetNPCPatrolPaths) NPC))
            (Utilities|Map|Add (Variables|Default|GetNPCPatrolPaths) NPC (Class|FortAthenaNpcPatrollingComponent|GetPatrolPath (Actor|GetComponentbyClass _asfort_athena_aibot_controller "/Script/FortniteAI.FortAthenaNpcPatrollingComponent") :!class /Script/FortniteAI.FortAthenaNpcPatrollingComponent)))
          (Class|FortAthenaNpcPatrollingComponent|RemovePatrolPath (Actor|GetComponentbyClass _asfort_athena_aibot_controller "/Script/FortniteAI.FortAthenaNpcPatrollingComponent")))))))

(fn OnConsumedByStorm
  (Fortnite|FortShippingLog self (Utilities|String|Append (Utilities|String|SelectString (Variables|/TidalCrane_BaseActivities/Gameplay/B_TidalCrane_BaseLocator.B_TidalCrane_BaseLocator_C|GetBaseName (Variables|Default|GetBaseLocator)) "<unknown base>" (Utilities|IsValid (Variables|Default|GetBaseLocator) :!class /Script/Engine.KismetSystemLibrary)) " has been consumed by the storm and will be disabled."))
  (for _array_element (Variables|Default|GetNPCs)
    (CallFunction|ApplyorRemoveEffectsonNPC true _array_element)
    (Ability|SendGameplayEventtoActor _array_element "(TagName=\"Event.NPC.TidalCrane.BaseDefense.Complete\")")))

(fn OnPlayerPresenceChanged)

(fn ProcessNPC
  (params (Actor :type Actor%20Object%20Reference))
  (Utilities|FlowControl|Sequence
    (:then_0
      (if (GameplayTags|HasAnyTags (Fortnite|Factions|GetFactionTagContainerforActor Actor) (Variables|NPCs|GetNPCFactionTags))
        (bind _asfort_player_pawn (Utilities|Casting|CastToFortPlayerPawn Actor))
        (return true _asfort_player_pawn)))
    (:then_1
      (return false 0))))

(fn RestoreNPCPatrolPath
  (params (NPC :type Fort%20Player%20Pawn%20Object%20Reference))
  (bind _npcpatrolpaths (Variables|Default|GetNPCPatrolPaths))
  (bind (_value _returnvalue) (Utilities|Map|Find _npcpatrolpaths NPC))
  ; Reset NPC Patrol path back to normal
  (if _returnvalue
    (bind _asfort_athena_aibot_controller (Utilities|Casting|CastToFortAthenaAIBotController (Pawn|GetController NPC)))
    (Utilities|IsValid (Actor|GetComponentbyClass _asfort_athena_aibot_controller "/Script/FortniteAI.FortAthenaNpcPatrollingComponent")
      (:Is%20Valid
        (Class|FortAthenaNpcPatrollingComponent|SetPatrolPath (Actor|GetComponentbyClass _asfort_athena_aibot_controller "/Script/FortniteAI.FortAthenaNpcPatrollingComponent") _value true)
        (Utilities|Map|Remove (Variables|Default|GetNPCPatrolPaths) NPC)
        (for _array_element (Variables|Default|GetNPC_CheeringSmartObjects)
          (SmartObject|SetSmartObjectEnabled _array_element))))))

(fn SendCompletionQuestEvent
  (params (OptionalSubjectInfo :type Gameplay%20Tag%20Container%20Structure))
  (bind _questparticipants (Variables|Default|GetQuestParticipants))
  (bind _optionalsubjectinfo (Variables|GetOptionalSubjectInfo))
  ;[0 0 0] Send a quest event to all players in the volume for when a base activity is completed. Optional subject tags can be used to identify a specific event
  (if (Utilities|Set|IsNotEmpty _questparticipants)
    (bind _valuepin (Utilities|Set|ForEachLoop(Set) _questparticipants))
    (LocalVariables|SetCustomVerbMessage (CustomVerb|MakeCustomVerbMessage _valuepin "(TagName=\"CustomVerb.BaseActivity.Complete\")"))
    (Utilities|FlowControl|Sequence
      (:then_0
        (if (Utilities|GameplayTagContainerNotEmpty _optionalsubjectinfo)
          (bind _result (CustomVerb|AddTagsforSubject (CustomVerb|MakeCustomVerbMessage _valuepin "(TagName=\"CustomVerb.BaseActivity.Complete\")") "Type" _optionalsubjectinfo))
          (LocalVariables|SetCustomVerbMessage _result)))
      (:then_1
        (CustomVerb|BroadcastCustomVerbMessage (LocalVariables|GetCustomVerbMessage))
        (if (not (Utilities|Array|ContainsItem (LocalVariables|GetTeamsGrantedPoints) (Team|GetActorTeam _valuepin)))
          (Utilities|Array|AddUnique (LocalVariables|GetTeamsGrantedPoints) (Team|GetActorTeam _valuepin))
          (Utilities|FlowControl|Sequence
            (:then_0
              (switch Utilities|FlowControl|Switch|SwitchonGameplayTagContainer _optionalsubjectinfo
                (:(GameplayTags=((TagName="Athena.Quests.BaseActivity.Defense")))
                  (LocalVariables|SetRankingContext "(TagName=\"CherrySmoke.HeroRank.Points.Context.BaseDefense\")")
                  (LocalVariables|SetPointsToAdd (ScalableFloat|GetValueatLevel (Variables|Default|GetRankPoints_Defense))))
                (:(GameplayTags=((TagName="Athena.Quests.BaseActivity.Reclaim")))
                  (LocalVariables|SetRankingContext "(TagName=\"CherrySmoke.HeroRank.Points.Context.BaseReclamation\")")
                  (LocalVariables|SetPointsToAdd (ScalableFloat|GetValueatLevel (Variables|Default|GetRankPoints_Reclamation))))))
            (:then_1
              (Class|CherrySmokeHeroRankingGlobalManager|AddHeroPointstoTeam (Actor|GetComponentbyClass (Fortnite|Athena|GameState|GetGameStateAthena) "/MilitaryRank/Components/MilitaryRank_GlobalManager.MilitaryRank_GlobalManager_C") (Team|GetActorTeam _valuepin) (LocalVariables|GetPointsToAdd) (LocalVariables|GetRankingContext) 0))))))))

(fn SetNPCLeashToReward
  (params (NPC :type Fort%20Player%20Pawn%20Object%20Reference))
  (if (not (Utilities|Map|Contains (Variables|Default|GetNPCPatrolPaths) NPC))
    (Utilities|IsValid NPC
      (:Is%20Valid
        (Utilities|IsValid (Variables|Default|GetRewardChestSpawnLocation)
          (:Is%20Valid
            (bind _asfort_athena_aibot_controller (Utilities|Casting|CastToFortAthenaAIBotController (Pawn|GetController NPC)))
            (Class|BPFLAICommonFunctions|SetRadialLeash _asfort_athena_aibot_controller (Transformation|GetActorLocation (Variables|Default|GetRewardChestSpawnLocation)) 7.5e+02 7.5e+02)
            (Utilities|IsValid (Actor|GetComponentbyClass _asfort_athena_aibot_controller "/Script/FortniteAI.FortAthenaNpcPatrollingComponent")
              (:Is%20Valid
                (Utilities|Map|Add (Variables|Default|GetNPCPatrolPaths) NPC (Class|FortAthenaNpcPatrollingComponent|GetPatrolPath (Actor|GetComponentbyClass _asfort_athena_aibot_controller "/Script/FortniteAI.FortAthenaNpcPatrollingComponent") :!class /Script/FortniteAI.FortAthenaNpcPatrollingComponent))
                (Class|FortAthenaNpcPatrollingComponent|RemovePatrolPath (Actor|GetComponentbyClass _asfort_athena_aibot_controller "/Script/FortniteAI.FortAthenaNpcPatrollingComponent"))
                (for _array_element (Variables|Default|GetNPC_CheeringSmartObjects)
                  (SmartObject|SetSmartObjectEnabled _array_element true))))))))))

(fn UpdateLootTagForAI
  (params (AIActor :type Actor%20Object%20Reference))
  (bind _ailoottag (Variables|Settings|GetAILootTag))
  (bind _returnvalue (GameplayTags|MakeGameplayTagContainerfromTag _ailoottag))
  (Networking|SwitchHasAuthority
    (:Authority
      (Utilities|IsValid AIActor
        (:Is%20Valid
          (if (Utilities|Set|IsEmpty (Variables|Default|GetOverlappingPlayers))
            (Ability|GameplayEffect|AddLooseGameplayTags AIActor _returnvalue)
            (else
              (Ability|GameplayEffect|RemoveLooseGameplayTags AIActor _returnvalue))))))))

(fn ConstructionScript)
