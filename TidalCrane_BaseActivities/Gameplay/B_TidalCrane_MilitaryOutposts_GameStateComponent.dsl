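; Server-authoritative GameState component that assigns TidalCrane military outposts to BaseDefense, Reclamation, Infestation or no event each match.
; Once the playlist is ready, it filters locators through a data table, applies hotfix overrides, then randomly fills hotfix-driven counts.
; Data table rows name each state's World Partition data layers, activated at runtime alongside minimap icons.
; BaseDefense layers activate on matching storm phase events, including a faked initial phase; cheat events support debugging.
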(event Custom|OnReadyToStart
  (params (Locators :type Array%20of%20B%20Tidal%20Crane%20Base%20Locator%20Object%20References) (SkipSafeZoneChecks :type Boolean))
  ;[0.01039 0.15 4.7e-05 0.5] ReadyToStart: Determine number of bases to activate and register for storm changes
  ;
  ; NOTE: This is called by the B_TidalCrane_BaseIdentifier placed in a map
  (Utilities|SwitchHasAuthority
    (:Authority
      (Class|FortAsyncActionCurrentPlaylistReady|PlaylistDataReadyAsync
        (:OnReady
          (if (not (Variables|Private|GetbInitialized))
            (Variables|Private|SetbInitialized true)
            (Variables|Private|SetbSkipSafeZoneCheck SkipSafeZoneChecks)
            (CallFunction|AddValidBaseLocators Locators)
            (Default|InitBaseData)
            (for _array_element (Variables|Private|GetBaseInstances)
              (Utilities|IsValid (Utilities|Struct|BreakTidalCraneBaseInstanceData _array_element)
                (:Is%20Valid
                  (Class|BTidalCraneBaseLocator|SetBaseType (Utilities|Struct|BreakTidalCraneBaseInstanceData _array_element) (.BaseState_9_31CFB2934F66745E4D31BEB61D6517D8 (Utilities|Struct|BreakTidalCraneBaseInstanceData _array_element)) :!class /TidalCrane_BaseActivities/Gameplay/B_TidalCrane_BaseLocator.B_TidalCrane_BaseLocator_C))))
            ;; Listen for Storm Phase Changes
            (if (Utilities|Array|IsNotEmpty (Variables|Private|GetBaseInstances))
              (bind _returnvalue (GameplayEventRouter|GetEventRouter "/Script/FortniteGame.FortGameState" (Components|GetOwner)))
              (bind (_asynctaskproxy _payload) (GameplayEventRouter|StartListeningToStatefulEvent _returnvalue "/Script/FortniteGame.FortSafeZonePhaseUpdatedEvent")
                (:then
                  (Variables|Private|SetPhaseStepUpdatedAsyncTask _asynctaskproxy)
                  (CallFunction|OnGamePhaseStepChange 0) ; Fake a storm phase event to trigger "start of match" BaseDefense locations
                )
                (:OnEventReceived
                  _)
                (:OnSavedEventStateExists
                  _))
              (CallFunction|OnGamePhaseStepChange (Utilities|Struct|BreakFortSafeZonePhaseUpdatedEvent _payload)))))))))

(event Custom|OnGamePhaseStepChange
  (params (PhaseIndex :type Integer))
  (bind _baseinstances (Variables|Private|GetBaseInstances))
  ;[0.008275 0.15 0.006295 0.5] Handle storm phase changes (potentially activate new bases)
  (for _index (range (Utilities|Array|Length (Variables|Private|GetBaseInstances)))
    (if (Math|Boolean|ANDBoolean (Utilities|Enum|Equal(Enum) (Variables|/TidalCrane_BaseActivities/Gameplay/B_TidalCrane_BaseLocator.B_TidalCrane_BaseLocator_C|GetBaseState (Utilities|Struct|BreakTidalCraneBaseInstanceData (Utilities|Array|Get(aref) _baseinstances _index))) "NewEnumerator0") (Utilities|Enum|Equal(Enum) (.BaseState_9_31CFB2934F66745E4D31BEB61D6517D8 (Utilities|Struct|BreakTidalCraneBaseInstanceData (Utilities|Array|Get(aref) _baseinstances _index))) "NewEnumerator1") (== (.StormPhase_8_7AAB9A3B49D593AB6640E788EEDCFECE (Utilities|Struct|BreakTidalCraneBaseInstanceData (Utilities|Array|Get(aref) _baseinstances _index))) PhaseIndex))
      (Default|ActivateBaseDefense (Utilities|Struct|BreakTidalCraneBaseInstanceData (Utilities|Array|Get(aref) _baseinstances _index)))
      (Utilities|Struct|SetmembersinTidalCraneBaseInstanceData (Utilities|Array|Get(aref) _baseinstances _index) -1))))

(event EventEndPlay
  (params (EndPlayReason :type EEndPlayReason%20Enum :default "Destroyed"))
  (bind _phasestepupdatedasynctask (Variables|Private|GetPhaseStepUpdatedAsyncTask))
  ;[0.001726 0.15 0.0297 0.5] EndPlay: Cleanup registrations
  (Utilities|IsValid _phasestepupdatedasynctask
    (:Is%20Valid
      (GameplayEventRouter|StopListeningtoStatefulEvent _phasestepupdatedasynctask))))

(event Custom|DebugTidalCraneState
  ; cheat ke * DebugTidalCraneState
  ;
  ; Print current state of base activations
  (Networking|SwitchHasAuthority
    (:Authority
      (CallFunction|DebugPrintCurrentState))))

(event Custom|DebugActivateTidalCraneBaseDefense
  ; cheat ke * DebugActivateTidalCraneBaseDefense
  ;
  ; Finds a non-activated BaseDefense base and activates it
  (Networking|SwitchHasAuthority
    (:Authority
      (CallFunction|DebugTriggerBaseDefenseEvent))))

(event Custom|DebugActivateAllTidalCraneBaseDefense
  (bind _baseinstances (Variables|Private|GetBaseInstances))
  ; cheat ke * DebugActivateAllTidalCraneBaseDefense
  ;
  ; Finds all non-activated BaseDefense base and activates them
  (Networking|SwitchHasAuthority
    (:Authority
      (for _index (range (Utilities|Array|Length (Variables|Private|GetBaseInstances)))
        (if (and (Utilities|Enum|Equal(Enum) (.BaseState_9_31CFB2934F66745E4D31BEB61D6517D8 (Utilities|Struct|BreakTidalCraneBaseInstanceData (Utilities|Array|Get(aref) _baseinstances _index))) "NewEnumerator1") (>= (.StormPhase_8_7AAB9A3B49D593AB6640E788EEDCFECE (Utilities|Struct|BreakTidalCraneBaseInstanceData (Utilities|Array|Get(aref) _baseinstances _index))) 0))
          (Default|ActivateBaseDefense (Utilities|Struct|BreakTidalCraneBaseInstanceData (Utilities|Array|Get(aref) _baseinstances _index)))
          (Utilities|Struct|SetmembersinTidalCraneBaseInstanceData (Utilities|Array|Get(aref) _baseinstances _index) -1))))))

(event Custom|DebugTidalCraneEnableLocations
  ;[0 0 0] Cheat ke * DebugTidalCraneEnableLocations
  ;
  ; Toggles the state of all bases
  (Networking|SwitchHasAuthority
    (:Authority
      (bind _basesdatalayerinfodatatable (Variables|Settings|GetBasesDataLayerInfoDataTable)
        (:then
          (bind _outrownames (DataTable|GetDataTableRowNames _basesdatalayerinfodatatable))
          (for _array_element _outrownames
            (bind _returnvalue (Utilities|GetDataTableRow _basesdatalayerinfodatatable _array_element))
            (if (ScalableFloat|GetValueasBool (Utilities|Struct|BreakTidalCraneBasesDataLayerInfo _returnvalue))
              (Utilities|IsValid (.BaseDefenseDataLayer_19_002E564440115750A04F79B9492C3BFA (Utilities|Struct|BreakTidalCraneBasesDataLayerInfo _returnvalue))
                (:Is%20Valid
                  (bind _returnvalue_2 (WorldPartition|GetDataLayerManager))
                  (Utilities|IsValid _returnvalue_2
                    (:Is%20Valid
                      (DataLayers|SetDataLayerRuntimeState _returnvalue_2 (.BaseDefenseDataLayer_19_002E564440115750A04F79B9492C3BFA (Utilities|Struct|BreakTidalCraneBasesDataLayerInfo _returnvalue)) "Activated" false :!class /Script/Engine.DataLayerManager) ; InUse
                    )))))))
        (:else)))))

(event Custom|DebugTidalCraneStormPhase
  (params (Phase :type Integer))
  ; cheat ke * DebugTidalCraneStormPhase #
  ;
  ; Fakes a StormPhase change event - useful for debugging in maps without GameState setup
  (Networking|SwitchHasAuthority
    (:Authority
      (CallFunction|OnGamePhaseStepChange Phase))))

(event Custom|ActivateTidalCraneBaseDefense
  (params (BaseName :type String))
  (bind _baseinstances (Variables|Private|GetBaseInstances))
  ; cheat ke * ActivateTidalCraneBaseDefense [BaseName]
  (Networking|SwitchHasAuthority
    (:Authority
      (for _index (range (Utilities|Array|Length (Variables|Private|GetBaseInstances)))
        (Utilities|IsValid (Utilities|Struct|BreakTidalCraneBaseInstanceData (Utilities|Array|Get(aref) _baseinstances _index))
          (:Is%20Valid
            (if (Utilities|String|Equal,CaseInsensitive(String) (Variables|/TidalCrane_BaseActivities/Gameplay/B_TidalCrane_BaseLocator.B_TidalCrane_BaseLocator_C|GetBaseName (Utilities|Struct|BreakTidalCraneBaseInstanceData (Utilities|Array|Get(aref) _baseinstances _index))) BaseName)
              (switch Utilities|FlowControl|Switch|SwitchonTidalCrane_BaseTypes (.BaseState_9_31CFB2934F66745E4D31BEB61D6517D8 (Utilities|Struct|BreakTidalCraneBaseInstanceData (Utilities|Array|Get(aref) _baseinstances _index)))
                (:NewEnumerator0
                  _)
                (:NewEnumerator1
                  _))
              (Default|ActivateBaseDefense (Utilities|Struct|BreakTidalCraneBaseInstanceData (Utilities|Array|Get(aref) _baseinstances _index)))
              (Utilities|Struct|SetmembersinTidalCraneBaseInstanceData (Utilities|Array|Get(aref) _baseinstances _index) "NewEnumerator1" -1))))))))

(fn ActivateBaseDefense
  (params (BaseLocator :type B%20Tidal%20Crane%20Base%20Locator%20Object%20Reference))
  (bind (_basedatalayerrow_datatable _basedatalayerrow_rowname) (Utilities|Struct|BreakDataTableRowHandle (Variables|/TidalCrane_BaseActivities/Gameplay/B_TidalCrane_BaseLocator.B_TidalCrane_BaseLocator_C|GetBaseDataLayerRow_DataTable BaseLocator)))
  (Utilities|IsValid BaseLocator
    (:Is%20Valid
      (Class|BTidalCraneBaseLocator|SetBaseMinimapIcon BaseLocator "NewEnumerator2") ; Set minimap state here so icon is not reliant on\nDataLayer load/BeginPlay (avoids PIE load issues)
      (bind _returnvalue (Utilities|GetDataTableRow _basedatalayerrow_datatable _basedatalayerrow_rowname))
      (Utilities|IsValid (.BaseDefenseDataLayer_19_002E564440115750A04F79B9492C3BFA (Utilities|Struct|BreakTidalCraneBasesDataLayerInfo _returnvalue))
        (:Is%20Valid
          (bind _returnvalue_2 (WorldPartition|GetDataLayerManager))
          (DataLayers|SetDataLayerRuntimeState _returnvalue_2 (.BaseDefenseDataLayer_19_002E564440115750A04F79B9492C3BFA (Utilities|Struct|BreakTidalCraneBasesDataLayerInfo _returnvalue)) "Activated" false :!class /Script/Engine.DataLayerManager)
          (Math|Integer|IncrementInt (Variables|Private|GetNumActivatedBaseDefense)
            (:%20%20
              (Fortnite|FortShippingLog self (Utilities|String|Append "Activated " (Utilities|String|ToString(Name) _basedatalayerrow_rowname) " for BaseDefense event.")))))))))

(fn ActivateBaseDefenseInitialDataLayers
  (params (BaseLocator :type B%20Tidal%20Crane%20Base%20Locator%20Object%20Reference))
  (bind (_basedatalayerrow_datatable _basedatalayerrow_rowname) (Utilities|Struct|BreakDataTableRowHandle (Variables|/TidalCrane_BaseActivities/Gameplay/B_TidalCrane_BaseLocator.B_TidalCrane_BaseLocator_C|GetBaseDataLayerRow_DataTable BaseLocator)))
  (Utilities|IsValid BaseLocator
    (:Is%20Valid
      (Class|BTidalCraneBaseLocator|ReserveDensityforBaseDefense BaseLocator) ; SpawnManager won't load until later,\nso pre-reserve density now to ensure there is room
      (bind _returnvalue (Utilities|GetDataTableRow _basedatalayerrow_datatable _basedatalayerrow_rowname))
      (Utilities|FlowControl|Sequence
        (:then_0
          ; Map DataLayer
          (Utilities|IsValid (.MapBaseDefenseDataLayer_44_9D287D8D43D07F7A8B25CE88D0B9CFEB (Utilities|Struct|BreakTidalCraneBasesDataLayerInfo _returnvalue))
            (:Is%20Valid
              (bind _returnvalue_2 (WorldPartition|GetDataLayerManager))
              (DataLayers|SetDataLayerRuntimeState _returnvalue_2 (.MapBaseDefenseDataLayer_44_9D287D8D43D07F7A8B25CE88D0B9CFEB (Utilities|Struct|BreakTidalCraneBasesDataLayerInfo _returnvalue)) "Activated" false :!class /Script/Engine.DataLayerManager))))
        (:then_1
          ; NPC Data Layer
          (Utilities|IsValid (.NPCStartOfMatchDataLayer_29_510F46814CB6FEEDA074308AFD4B5938 (Utilities|Struct|BreakTidalCraneBasesDataLayerInfo _returnvalue))
            (:Is%20Valid
              (bind _returnvalue_3 (WorldPartition|GetDataLayerManager))
              (DataLayers|SetDataLayerRuntimeState _returnvalue_3 (.NPCStartOfMatchDataLayer_29_510F46814CB6FEEDA074308AFD4B5938 (Utilities|Struct|BreakTidalCraneBasesDataLayerInfo _returnvalue)) "Activated" false :!class /Script/Engine.DataLayerManager))))
        (:then_2
          ; Marketplace Data Layer
          (Utilities|IsValid (.MarketDataLayer_36_118DCD8B4F9C6387F715FE90FE8060B1 (Utilities|Struct|BreakTidalCraneBasesDataLayerInfo _returnvalue))
            (:Is%20Valid
              (bind _returnvalue_4 (WorldPartition|GetDataLayerManager))
              (DataLayers|SetDataLayerRuntimeState _returnvalue_4 (.MarketDataLayer_36_118DCD8B4F9C6387F715FE90FE8060B1 (Utilities|Struct|BreakTidalCraneBasesDataLayerInfo _returnvalue)) "Activated" false :!class /Script/Engine.DataLayerManager))))
        (:then_3
          ; UI Icon
          (Class|BTidalCraneBaseLocator|SetBaseMinimapIcon (Variables|GetBaseLocator) "NewEnumerator0"))))))

(fn ActivateDataLayers
  (params (ValidActors :type Array%20of%20Actor%20Object%20References))
  (Networking|SwitchHasAuthority
    (:Authority
      (bind _returnvalue (WorldPartition|GetDataLayerManager))
      (Utilities|IsValid _returnvalue
        (:Is%20Valid
          (for _array_element ValidActors
            (bind _asb_tidal_crane_base_locator (Utilities|Casting|CastToB_TidalCrane_BaseLocator _array_element))
            (bind (_basedatalayerrow_datatable _basedatalayerrow_rowname) (Utilities|Struct|BreakDataTableRowHandle (Variables|/TidalCrane_BaseActivities/Gameplay/B_TidalCrane_BaseLocator.B_TidalCrane_BaseLocator_C|GetBaseDataLayerRow_DataTable _asb_tidal_crane_base_locator)))
            (bind _returnvalue_2 (Utilities|GetDataTableRow _basedatalayerrow_datatable _basedatalayerrow_rowname))
            (bind _returnvalue_3 (WorldPartition|GetDataLayerManager))
            (DataLayers|SetDataLayerRuntimeState _returnvalue_3 (.BaseDefenseDataLayer_19_002E564440115750A04F79B9492C3BFA (Utilities|Struct|BreakTidalCraneBasesDataLayerInfo _returnvalue_2)) "Activated" false :!class /Script/Engine.DataLayerManager)))
        (:Is%20Not%20Valid
          (Fortnite|FortShippingLog self "FinalizeDataLayerRuntimeState: DataLayerManager is null." true))))))

(fn ActivateDefaultBaseDataLayers
  (params (BaseLocator :type B%20Tidal%20Crane%20Base%20Locator%20Object%20Reference))
  (bind (_basedatalayerrow_datatable _basedatalayerrow_rowname) (Utilities|Struct|BreakDataTableRowHandle (Variables|/TidalCrane_BaseActivities/Gameplay/B_TidalCrane_BaseLocator.B_TidalCrane_BaseLocator_C|GetBaseDataLayerRow_DataTable BaseLocator)))
  (bind (_basedatalayerrow_datatable_2 _basedatalayerrow_rowname_2) (Utilities|Struct|BreakDataTableRowHandle (Variables|/TidalCrane_BaseActivities/Gameplay/B_TidalCrane_BaseLocator.B_TidalCrane_BaseLocator_C|GetBaseDataLayerRow_DataTable (Variables|GetBaseLocator))))
  (Utilities|IsValid BaseLocator
    (:Is%20Valid
      (bind _returnvalue (Utilities|GetDataTableRow _basedatalayerrow_datatable _basedatalayerrow_rowname))
      (Utilities|FlowControl|Sequence
        (:then_0
          ; Map DataLayer
          (Utilities|IsValid (.MapBaseDefenseDataLayer_44_9D287D8D43D07F7A8B25CE88D0B9CFEB (Utilities|Struct|BreakTidalCraneBasesDataLayerInfo _returnvalue))
            (:Is%20Valid
              (bind _returnvalue_2 (WorldPartition|GetDataLayerManager))
              (DataLayers|SetDataLayerRuntimeState _returnvalue_2 (.MapBaseDefenseDataLayer_44_9D287D8D43D07F7A8B25CE88D0B9CFEB (Utilities|Struct|BreakTidalCraneBasesDataLayerInfo _returnvalue)) "Activated" false :!class /Script/Engine.DataLayerManager))))
        (:then_1
          ; NPC Data Layer
          (Utilities|IsValid (.NPCStartOfMatchDataLayer_29_510F46814CB6FEEDA074308AFD4B5938 (Utilities|Struct|BreakTidalCraneBasesDataLayerInfo _returnvalue))
            (:Is%20Valid
              (bind _returnvalue_3 (WorldPartition|GetDataLayerManager))
              (DataLayers|SetDataLayerRuntimeState _returnvalue_3 (.NPCStartOfMatchDataLayer_29_510F46814CB6FEEDA074308AFD4B5938 (Utilities|Struct|BreakTidalCraneBasesDataLayerInfo _returnvalue)) "Activated" false :!class /Script/Engine.DataLayerManager))))
        (:then_2
          ; Marketplace Data Layer
          (Utilities|IsValid (.MarketDataLayer_36_118DCD8B4F9C6387F715FE90FE8060B1 (Utilities|Struct|BreakTidalCraneBasesDataLayerInfo _returnvalue))
            (:Is%20Valid
              (bind _returnvalue_4 (WorldPartition|GetDataLayerManager))
              (DataLayers|SetDataLayerRuntimeState _returnvalue_4 (.MarketDataLayer_36_118DCD8B4F9C6387F715FE90FE8060B1 (Utilities|Struct|BreakTidalCraneBasesDataLayerInfo _returnvalue)) "Activated" false :!class /Script/Engine.DataLayerManager))))
        (:then_3
          ; UI Icon
          (Class|BTidalCraneBaseLocator|SetBaseMinimapIcon (Variables|GetBaseLocator) "NewEnumerator0"))
        (:then_4
          ; Print to Log
          (Fortnite|FortShippingLog self (Utilities|String|Append "Activated " (Utilities|String|ToString(Name) _basedatalayerrow_rowname_2) " for default Base (no event selected).")))))))

(fn ActivateInfestation
  (params (BaseLocator :type B%20Tidal%20Crane%20Base%20Locator%20Object%20Reference))
  (bind (_basedatalayerrow_datatable _basedatalayerrow_rowname) (Utilities|Struct|BreakDataTableRowHandle (Variables|/TidalCrane_BaseActivities/Gameplay/B_TidalCrane_BaseLocator.B_TidalCrane_BaseLocator_C|GetBaseDataLayerRow_DataTable BaseLocator)))
  (bind (_basedatalayerrow_datatable_2 _basedatalayerrow_rowname_2) (Utilities|Struct|BreakDataTableRowHandle (Variables|/TidalCrane_BaseActivities/Gameplay/B_TidalCrane_BaseLocator.B_TidalCrane_BaseLocator_C|GetBaseDataLayerRow_DataTable (Variables|GetBaseLocator))))
  (Utilities|IsValid BaseLocator
    (:Is%20Valid
      (bind _returnvalue (Utilities|GetDataTableRow _basedatalayerrow_datatable _basedatalayerrow_rowname))
      (Utilities|FlowControl|Sequence
        (:then_0
          ; Map DataLayer
          (Utilities|IsValid (.MapInfestationDataLayer_46_D1B33AFB4D5D8CC3E5312CA5B035EE55 (Utilities|Struct|BreakTidalCraneBasesDataLayerInfo _returnvalue))
            (:Is%20Valid
              (bind _returnvalue_2 (WorldPartition|GetDataLayerManager))
              (DataLayers|SetDataLayerRuntimeState _returnvalue_2 (.MapInfestationDataLayer_46_D1B33AFB4D5D8CC3E5312CA5B035EE55 (Utilities|Struct|BreakTidalCraneBasesDataLayerInfo _returnvalue)) "Activated" false :!class /Script/Engine.DataLayerManager))))
        (:then_1
          ; Load Infested Data Layer
          (Utilities|IsValid (.InfestedDataLayer_20_50F99B9147147944CB91208004043FAA (Utilities|Struct|BreakTidalCraneBasesDataLayerInfo _returnvalue))
            (:Is%20Valid
              (bind _returnvalue_3 (WorldPartition|GetDataLayerManager))
              (DataLayers|SetDataLayerRuntimeState _returnvalue_3 (.InfestedDataLayer_20_50F99B9147147944CB91208004043FAA (Utilities|Struct|BreakTidalCraneBasesDataLayerInfo _returnvalue)) "Activated" false :!class /Script/Engine.DataLayerManager))))
        (:then_2
          ; Infested Marketplace Data Layer
          (Utilities|IsValid (.InfestedMarketDataLayer_38_848B5B45438E53BA98EC799CB9D8A522 (Utilities|Struct|BreakTidalCraneBasesDataLayerInfo _returnvalue))
            (:Is%20Valid
              (bind _returnvalue_4 (WorldPartition|GetDataLayerManager))
              (DataLayers|SetDataLayerRuntimeState _returnvalue_4 (.InfestedMarketDataLayer_38_848B5B45438E53BA98EC799CB9D8A522 (Utilities|Struct|BreakTidalCraneBasesDataLayerInfo _returnvalue)) "Activated" false :!class /Script/Engine.DataLayerManager))))
        (:then_3
          ; UI Icon
          (Class|BTidalCraneBaseLocator|SetBaseMinimapIcon (Variables|GetBaseLocator) "NewEnumerator3"))
        (:then_4
          (Math|Integer|IncrementInt (Variables|Private|GetNumInfestations)))
        (:then_5
          (Fortnite|FortShippingLog self (Utilities|String|Append "Activated " (Utilities|String|ToString(Name) _basedatalayerrow_rowname_2) " for Infestation event.")))))))

(fn ActivateReclamation
  (params (BaseLocator :type B%20Tidal%20Crane%20Base%20Locator%20Object%20Reference))
  (bind (_basedatalayerrow_datatable _basedatalayerrow_rowname) (Utilities|Struct|BreakDataTableRowHandle (Variables|/TidalCrane_BaseActivities/Gameplay/B_TidalCrane_BaseLocator.B_TidalCrane_BaseLocator_C|GetBaseDataLayerRow_DataTable (Variables|GetBaseLocator))))
  (bind (_basedatalayerrow_datatable_2 _basedatalayerrow_rowname_2) (Utilities|Struct|BreakDataTableRowHandle (Variables|/TidalCrane_BaseActivities/Gameplay/B_TidalCrane_BaseLocator.B_TidalCrane_BaseLocator_C|GetBaseDataLayerRow_DataTable BaseLocator)))
  (Utilities|IsValid BaseLocator
    (:Is%20Valid
      (bind _returnvalue (Utilities|GetDataTableRow _basedatalayerrow_datatable_2 _basedatalayerrow_rowname_2))
      (Utilities|FlowControl|Sequence
        (:then_0
          ; Map DataLayer
          (Utilities|IsValid (.MapInfestationDataLayer_46_D1B33AFB4D5D8CC3E5312CA5B035EE55 (Utilities|Struct|BreakTidalCraneBasesDataLayerInfo _returnvalue))
            (:Is%20Valid
              (bind _returnvalue_2 (WorldPartition|GetDataLayerManager))
              (DataLayers|SetDataLayerRuntimeState _returnvalue_2 (.MapInfestationDataLayer_46_D1B33AFB4D5D8CC3E5312CA5B035EE55 (Utilities|Struct|BreakTidalCraneBasesDataLayerInfo _returnvalue)) "Activated" false :!class /Script/Engine.DataLayerManager))))
        (:then_1
          ; Load Infested base visuals
          (Utilities|IsValid (.InfestedDataLayer_20_50F99B9147147944CB91208004043FAA (Utilities|Struct|BreakTidalCraneBasesDataLayerInfo _returnvalue))
            (:Is%20Valid
              (bind _returnvalue_3 (WorldPartition|GetDataLayerManager))
              (DataLayers|SetDataLayerRuntimeState _returnvalue_3 (.InfestedDataLayer_20_50F99B9147147944CB91208004043FAA (Utilities|Struct|BreakTidalCraneBasesDataLayerInfo _returnvalue)) "Activated" false :!class /Script/Engine.DataLayerManager))))
        (:then_2
          ; Load Reclamation layer
          (Utilities|IsValid (.ReclamationDataLayer_21_0C8961A04FAD73C54CC10488060C39EA (Utilities|Struct|BreakTidalCraneBasesDataLayerInfo _returnvalue))
            (:Is%20Valid
              (bind _returnvalue_4 (WorldPartition|GetDataLayerManager))
              (DataLayers|SetDataLayerRuntimeState _returnvalue_4 (.ReclamationDataLayer_21_0C8961A04FAD73C54CC10488060C39EA (Utilities|Struct|BreakTidalCraneBasesDataLayerInfo _returnvalue)) "Activated" false :!class /Script/Engine.DataLayerManager))))
        (:then_3
          ; Load Reclamation NPCs
          (Utilities|IsValid (.NPCRescueDataLayer_30_2A87E32B4950D42B663D61B54C4D7B6C (Utilities|Struct|BreakTidalCraneBasesDataLayerInfo _returnvalue))
            (:Is%20Valid
              (bind _returnvalue_5 (WorldPartition|GetDataLayerManager))
              (DataLayers|SetDataLayerRuntimeState _returnvalue_5 (.NPCRescueDataLayer_30_2A87E32B4950D42B663D61B54C4D7B6C (Utilities|Struct|BreakTidalCraneBasesDataLayerInfo _returnvalue)) "Activated" false :!class /Script/Engine.DataLayerManager))))
        (:then_4
          ; Load Infested Marketplace
          (Utilities|IsValid (.InfestedMarketDataLayer_38_848B5B45438E53BA98EC799CB9D8A522 (Utilities|Struct|BreakTidalCraneBasesDataLayerInfo _returnvalue))
            (:Is%20Valid
              (bind _returnvalue_6 (WorldPartition|GetDataLayerManager))
              (DataLayers|SetDataLayerRuntimeState _returnvalue_6 (.InfestedMarketDataLayer_38_848B5B45438E53BA98EC799CB9D8A522 (Utilities|Struct|BreakTidalCraneBasesDataLayerInfo _returnvalue)) "Activated" false :!class /Script/Engine.DataLayerManager))))
        (:then_5
          ; Minimap Icon
          (Class|BTidalCraneBaseLocator|SetBaseMinimapIcon (Variables|GetBaseLocator) "NewEnumerator4"))
        (:then_6
          (Math|Integer|IncrementInt (Variables|Private|GetNumReclamations)))
        (:then_7
          (Fortnite|FortShippingLog self (Utilities|String|Append "Activated " (Utilities|String|ToString(Name) _basedatalayerrow_rowname) " for Reclamation event.")))))))

(fn AddValidBaseLocators
  (params (BaseLocators :type Array%20of%20B%20Tidal%20Crane%20Base%20Locator%20Object%20References))
  (for _array_element BaseLocators
    (bind (_basedatalayerrow_datatable _basedatalayerrow_rowname) (Utilities|Struct|BreakDataTableRowHandle (Variables|/TidalCrane_BaseActivities/Gameplay/B_TidalCrane_BaseLocator.B_TidalCrane_BaseLocator_C|GetBaseDataLayerRow_DataTable _array_element)))
    (bind _returnvalue (Utilities|GetDataTableRow _basedatalayerrow_datatable _basedatalayerrow_rowname))
    (if (ScalableFloat|GetValueasBool (Utilities|Struct|BreakTidalCraneBasesDataLayerInfo _returnvalue))
      (Utilities|Array|Add (Variables|Private|GetAvailableBases) _array_element))))

(fn ApplyOverrides
  (if (ScalableFloat|GetValueasBool (Variables|Hotfix|GetHFRow%20AllowOverrideActivity))
    (bind (_arrayindex _arrayelement) (Utilities|Array|ReverseforEachLoop (Variables|Private|GetAvailableBases))
      (:LoopBody
        (Utilities|IsValid _arrayelement
          (:Is%20Valid
            (bind (_basedatalayerrow_datatable _basedatalayerrow_rowname) (Utilities|Struct|BreakDataTableRowHandle (Variables|/TidalCrane_BaseActivities/Gameplay/B_TidalCrane_BaseLocator.B_TidalCrane_BaseLocator_C|GetBaseDataLayerRow_DataTable _arrayelement)))
            (bind _returnvalue (Utilities|GetDataTableRow _basedatalayerrow_datatable _basedatalayerrow_rowname))
            (if (and (ScalableFloat|GetValueasBool (Utilities|Struct|BreakTidalCraneBasesDataLayerInfo _returnvalue)) (Utilities|Enum|NotEqual(Enum) (.OverrideActivity_33_A653E9D447F16794D7CBF1ABB0647C3D (Utilities|Struct|BreakTidalCraneBasesDataLayerInfo _returnvalue)) "NewEnumerator4"))
              (Fortnite|FortShippingLog self (Utilities|String|Append "Applying " (Utilities|String|EnumtoString (.OverrideActivity_33_A653E9D447F16794D7CBF1ABB0647C3D (Utilities|Struct|BreakTidalCraneBasesDataLayerInfo _returnvalue))) " override for " (Variables|/TidalCrane_BaseActivities/Gameplay/B_TidalCrane_BaseLocator.B_TidalCrane_BaseLocator_C|GetBaseName _arrayelement)))
              (switch Utilities|FlowControl|Switch|SwitchonTidalCrane_BaseTypes (.OverrideActivity_33_A653E9D447F16794D7CBF1ABB0647C3D (Utilities|Struct|BreakTidalCraneBasesDataLayerInfo _returnvalue))
                (:NewEnumerator0
                  (Default|ActivateDefaultBaseDataLayers _arrayelement)
                  (CallFunction|SetStartupBaseInfo _arrayelement "NewEnumerator0" -1))
                (:NewEnumerator1
                  (Default|ActivateBaseDefenseInitialDataLayers _arrayelement)
                  (CallFunction|SetStartupBaseInfo _arrayelement "NewEnumerator1")
                  (Variables|Private|SetBaseDefensesToSelect (Math|Integer|Max(Integer) (- (Variables|Private|GetBaseDefensesToSelect) 1)))
                  (Math|Integer|IncrementInt (Variables|Private|GetNumDefenses)))
                (:NewEnumerator2
                  (CallFunction|ActivateInfestation _arrayelement)
                  (CallFunction|SetStartupBaseInfo _arrayelement "NewEnumerator2" -1)
                  (Variables|Private|SetInfestationsToSelect (Math|Integer|Max(Integer) (- (Variables|Private|GetInfestationsToSelect) 1))))
                (:NewEnumerator3
                  (CallFunction|ActivateReclamation _arrayelement)
                  (CallFunction|SetStartupBaseInfo _arrayelement "NewEnumerator3" -1)
                  (Variables|Private|SetReclamationsToSelect (Math|Integer|Max(Integer) (- (Variables|Private|GetReclamationsToSelect) 1)))))))))) ; Iterate backwards so we can safely remove bases
  ))

(fn ChooseBasesForBaseDefense
  (params (NumBases :type Integer))
  (bind _validbases (LocalVariables|GetValidBases))
  (bind _output_get (LocalVariables|SetValidBases (Variables|Private|GetAvailableBases)))
  (Utilities|Array|Shuffle _output_get)
  (bind _index (Utilities|FlowControl|ForLoopwithBreak 0 (- NumBases 1))
    (:LoopBody
      (if (>= _index (Utilities|Array|Length _validbases))
        (Fortnite|FortShippingLog self "Error: Not enough bases for BaseDefense events!" true))
      (Utilities|IsValid (Utilities|Array|Get(acopy) _validbases _index)
        (:Is%20Valid
          (Default|ActivateBaseDefenseInitialDataLayers (Utilities|Array|Get(acopy) _validbases _index))
          (CallFunction|SetStartupBaseInfo (Utilities|Array|Get(acopy) _validbases _index) "NewEnumerator1")
          (Math|Integer|IncrementInt (Variables|Private|GetNumDefenses)))))))

(fn Debug_PrintCurrentState
  (Fortnite|FortShippingLog self (Utilities|String|Append "BaseDefense: " (Utilities|String|ToString(Integer) (Variables|Private|GetNumActivatedBaseDefense)) "/" (Utilities|String|ToString(Integer) (Variables|Private|GetNumDefenses)) " . Reclamation: " (Utilities|String|ToString(Integer) (Variables|Private|GetNumReclamations)) ". Infestations: " (Utilities|String|ToString(Integer) (Variables|Private|GetNumInfestations))))
  (for _array_element (Variables|Private|GetBaseInstances)
    (bind (_array_element_baselocator_2_c88139c04c7c90dde7cf11ad22e65952 _array_element_basestate_9_31cfb2934f66745e4d31beb61d6517d8 _array_element_stormphase_8_7aab9a3b49d593ab6640e788eedcfece) (Utilities|Struct|BreakTidalCraneBaseInstanceData _array_element))
    (Utilities|IsValid _array_element_baselocator_2_c88139c04c7c90dde7cf11ad22e65952
      (:Is%20Valid
        (Fortnite|FortShippingLog self (Utilities|String|Append (Utilities|String|SelectString "<unknown>" (Variables|/TidalCrane_BaseActivities/Gameplay/B_TidalCrane_BaseLocator.B_TidalCrane_BaseLocator_C|GetBaseName _array_element_baselocator_2_c88139c04c7c90dde7cf11ad22e65952) (Utilities|String|IsEmpty (Variables|/TidalCrane_BaseActivities/Gameplay/B_TidalCrane_BaseLocator.B_TidalCrane_BaseLocator_C|GetBaseName _array_element_baselocator_2_c88139c04c7c90dde7cf11ad22e65952))) ": " (Utilities|String|EnumtoString _array_element_basestate_9_31cfb2934f66745e4d31beb61d6517d8) " " (Utilities|String|SelectString (Utilities|String|Append "StormPhase=" (Utilities|String|ToString(Integer) _array_element_stormphase_8_7aab9a3b49d593ab6640e788eedcfece)) "" (>= _array_element_stormphase_8_7aab9a3b49d593ab6640e788eedcfece 0))))))))

(fn Debug_TriggerBaseDefenseEvent
  (bind _validbaseindexes (LocalVariables|GetValidBaseIndexes))
  (bind _returnvalue (Utilities|Array|Length _validbaseindexes))
  (bind _returnvalue_2 (- _returnvalue 1))
  (bind _returnvalue_3 (Math|Random|RandomIntegerinRange 0 _returnvalue_2))
  (bind _baseinstances (Variables|Private|GetBaseInstances))
  (bind _output (Utilities|Array|Get(aref) _baseinstances _returnvalue_3))
  (Networking|SwitchHasAuthority
    (:Authority
      ; Pick a random base assigned to BaseDefense that isn't already active and activate it
      (for _array_element (Variables|Private|GetBaseInstances)
        (bind (_array_element_baselocator_2_c88139c04c7c90dde7cf11ad22e65952 _array_element_basestate_9_31cfb2934f66745e4d31beb61d6517d8 _array_element_stormphase_8_7aab9a3b49d593ab6640e788eedcfece) (Utilities|Struct|BreakTidalCraneBaseInstanceData _array_element))
        (if (and (Utilities|Enum|Equal(Enum) _array_element_basestate_9_31cfb2934f66745e4d31beb61d6517d8 "NewEnumerator1") (> _array_element_stormphase_8_7aab9a3b49d593ab6640e788eedcfece 0))
          (Utilities|Array|Add (LocalVariables|GetValidBaseIndexes) _array_element_stormphase_8_7aab9a3b49d593ab6640e788eedcfece)))
      (if (> _returnvalue 0)
        (Default|ActivateBaseDefense (Utilities|Struct|BreakTidalCraneBaseInstanceData _output))
        (Utilities|Struct|SetmembersinTidalCraneBaseInstanceData _output -1)
        (return)))))

(fn GetMinMaxRangeValidated
  (params (MinValue :type Scalable%20Float%20Structure) (MaxValue :type Scalable%20Float%20Structure) (DebugIdentifier :type String))
  (LocalVariables|SetMinVal (ScalableFloat|GetValueasInteger MinValue))
  (LocalVariables|SetMaxVal (ScalableFloat|GetValueasInteger MaxValue))
  (if (Math|Boolean|ANDBoolean (<= (LocalVariables|GetMinVal) (LocalVariables|GetMaxVal)) (>= (LocalVariables|GetMinVal) 0) (>= (LocalVariables|GetMaxVal) 0))
    (return (LocalVariables|GetMinVal) (LocalVariables|GetMaxVal) true)
    (else
      (Fortnite|FortShippingLog self (Utilities|String|Append "Error: Invalid base range for  " (Variables|GetDebugIdentifier) ". " (Utilities|String|ToString(Integer) (LocalVariables|GetMinVal)) " - " (Utilities|String|ToString(Integer) (LocalVariables|GetMaxVal)) ". Activation will be skipped!") true)
      (return 0 0 false))))

(fn InitBaseData
  (bind _availablebases (Variables|Private|GetAvailableBases))
  (bind _output (Utilities|Array|Get(acopy) _availablebases))
  (bind _minreclamations (LocalVariables|GetMinReclamations))
  (bind _mininfestations (LocalVariables|GetMinInfestations))
  (bind _minbasedefense (LocalVariables|GetMinBaseDefense))
  (bind _returnvalue (Utilities|Operators|Add _minbasedefense _mininfestations _minreclamations))
  (bind _availablebases_2 (Variables|Private|GetAvailableBases))
  (bind _availablebases_3 (Variables|Private|GetAvailableBases))
  (CallFunction|RegisterValidBasesDataLayer) ; Checking from the available data layers
  (Utilities|FlowControl|Sequence
    (:then_0
      (bind (_min _max) (CallFunction|GetMinMaxRangeValidated (Variables|Hotfix|GetHFRow%20MinBaseDefenses) (Variables|Hotfix|GetHFRow%20MaxBaseDefenses) "BaseDefense"))
      (LocalVariables|SetMinBaseDefense _min)
      (LocalVariables|SetMaxBaseDefense _max))
    (:then_1
      (bind (_min_2 _max_2) (CallFunction|GetMinMaxRangeValidated (Variables|Hotfix|GetHFRow%20MinInfestations) (Variables|Hotfix|GetHFRow%20MaxInfestations) "BaseInfestations"))
      (LocalVariables|SetMinInfestations _min_2)
      (LocalVariables|SetMaxInfestations _max_2))
    (:then_2
      (bind (_min_3 _max_3) (CallFunction|GetMinMaxRangeValidated (Variables|Hotfix|GetHFRow%20MinReclamations) (Variables|Hotfix|GetHFRow%20MaxReclamations) "BaseReclamations"))
      (LocalVariables|SetMinReclamations _min_3)
      (LocalVariables|SetMaxReclamations _max_3))
    (:then_3
      (LocalVariables|SetNumAvailableBases (Utilities|Array|Length (Variables|Private|GetAvailableBases)))
      (if (not (>= (LocalVariables|GetNumAvailableBases) _returnvalue))
        (Fortnite|FortShippingLog self (Utilities|String|Append "Error: Not enough available bases to activate minimum of " (Utilities|String|ToString(Integer) _returnvalue) "! Activation counts will not match.") true))
      (Variables|Private|SetBaseDefensesToSelect (Math|Random|RandomIntegerinRange (LocalVariables|GetMinBaseDefense) (LocalVariables|GetMaxBaseDefense)))
      (Variables|Private|SetReclamationsToSelect (Math|Random|RandomIntegerinRange (LocalVariables|GetMaxReclamations) (LocalVariables|GetMaxReclamations)))
      (Variables|Private|SetInfestationsToSelect (Math|Random|RandomIntegerinRange (LocalVariables|GetMinInfestations) (LocalVariables|GetMaxInfestations)))
      (Fortnite|FortShippingLog self (Utilities|String|Append "TidalCrane Base Initialization: Choosing " (Utilities|String|ToString(Integer) (Variables|Private|GetBaseDefensesToSelect)) " BaseDefenses. " (Utilities|String|ToString(Integer) (Variables|Private|GetReclamationsToSelect)) " Reclamations. " (Utilities|String|ToString(Integer) (Variables|Private|GetInfestationsToSelect)) " Infestations."))
      ; First apply any base overrides
      ; (if overrides are enabled)
      (Default|ApplyOverrides)
      ; Choose BaseDefense bases first since they start
      ; mid-match and need safe storm locations
      (CallFunction|ChooseBasesforBaseDefense (Variables|Private|GetBaseDefensesToSelect))
      (Utilities|FlowControl|Sequence
        (:then_0
          ; Choose Reclamation Bases
          (Utilities|FlowControl|ForLoopwithBreak 0 (- (Variables|Private|GetReclamationsToSelect) 1)
            (:LoopBody
              (if (Utilities|Array|IsEmpty (Variables|Private|GetAvailableBases))
                _)
              (bind _output_get (LocalVariables|SetIndex (Math|Random|RandomIntegerinRange 0 (- (Utilities|Array|Length (Variables|Private|GetAvailableBases)) 1))))
              (CallFunction|ActivateReclamation (Utilities|Array|Get(acopy) _availablebases_2 _output_get))
              (CallFunction|SetStartupBaseInfo (Utilities|Array|Get(acopy) _availablebases_2 _output_get) "NewEnumerator3" -1))))
        (:then_1
          ; Choose Infested bases
          (Utilities|FlowControl|ForLoopwithBreak 0 (- (Variables|Private|GetInfestationsToSelect) 1)
            (:LoopBody
              (if (Utilities|Array|IsEmpty (Variables|Private|GetAvailableBases))
                _)
              (bind _output_get_2 (LocalVariables|SetIndex (Math|Random|RandomIntegerinRange 0 (- (Utilities|Array|Length (Variables|Private|GetAvailableBases)) 1))))
              (CallFunction|ActivateInfestation (Utilities|Array|Get(acopy) _availablebases_3 _output_get_2))
              (CallFunction|SetStartupBaseInfo (Utilities|Array|Get(acopy) _availablebases_3 _output_get_2) "NewEnumerator2" -1))))
        (:then_2
          ; All remaining Bases have no event. Flag and enable DataLayer
          (while (Utilities|Array|IsNotEmpty (Variables|Private|GetAvailableBases))
            (Default|ActivateDefaultBaseDataLayers _output)
            (CallFunction|SetStartupBaseInfo _output "NewEnumerator0" -1)))))))

(fn RegisterValidBases_DataLayer
  (bind _basesdatalayerinfodatatable (Variables|Settings|GetBasesDataLayerInfoDataTable)
    (:then
      (bind _outrownames (DataTable|GetDataTableRowNames _basesdatalayerinfodatatable))
      (for _array_element _outrownames
        (bind _returnvalue (Utilities|GetDataTableRow _basesdatalayerinfodatatable _array_element))
        (if (ScalableFloat|GetValueasBool (Utilities|Struct|BreakTidalCraneBasesDataLayerInfo _returnvalue))
          (Utilities|Array|AddUnique (Variables|Private|GetHotfix%20Enabled%20Data%20Layer%20Row%20Entries) _array_element)
          (Utilities|Array|AddUnique (Variables|Private|GetIn%20Use) (.BaseDefenseDataLayer_19_002E564440115750A04F79B9492C3BFA (Utilities|Struct|BreakTidalCraneBasesDataLayerInfo _returnvalue))) ; Storing the data layer for each base on the map
        )))
    (:else)))

(fn SetStartupBaseInfo
  (params (BaseLocator :type B%20Tidal%20Crane%20Base%20Locator%20Object%20Reference) (InitialState :type TidalCrane_BaseTypes%20Enum :default "NewEnumerator0") (StormPhase :type Integer))
  (Utilities|Array|RemoveItem (Variables|Private|GetAvailableBases) BaseLocator)
  (Utilities|Array|Add (Variables|Private|GetBaseInstances) (Utilities|Struct|MakeTidalCraneBaseInstanceData BaseLocator InitialState StormPhase)))
