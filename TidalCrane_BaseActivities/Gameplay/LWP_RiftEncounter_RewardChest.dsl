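; Rift encounter reward chest placement actor that spawns the configured reward chest in the world.
; When its post-spawn event fires, an authority check limits spawning to the server.
; It spawns the reward chest class at its own transform, ignoring collisions and passing its instigator, then stores the reference in Reward Chest.
; BeginPlay only calls the parent, and the overlap and tick events are empty.
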(event EventBeginPlay
  (|CallParentFunction))

(event Collision|EventActorBeginOverlap
  (params (OtherActor :type Actor%20Object%20Reference)))

(event EventTick
  (params (DeltaSeconds :type Float%20%28single-precision%29)))

(event EventPostSpawnEvents
  (Networking|SwitchHasAuthority
    (:Authority
      (bind _returnvalue (Game|SpawnActorfromClass (Variables|Default|GetReward%20chest%20class) (Transformation|GetActorTransform) "AlwaysSpawn" "MultiplyWithRoot" 0 (Variables|Actor|GetInstigator)))
      (Variables|Default|SetReward%20Chest _returnvalue))))

(fn ConstructionScript
  (|CallParentFunction))
